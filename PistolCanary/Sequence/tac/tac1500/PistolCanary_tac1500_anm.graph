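# Level Sequence for the animation layer of shot tac1500 in PistolCanary, animating the tentacle boss.
# It possesses a TentacleBoss actor through a single binding with one skeletal animation track.
# Two animation sections on that track overlap briefly where the first hands over to the second.
# It has no root-level tracks, so it carries only the boss animation, at a 60 fps display rate.
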LevelSequence /PistolCanary/Sequence/tac/tac1500/PistolCanary_tac1500_anm
bindings [TentacleBoss {
  id=#1
  possessed_class=PistolCanary_TentacleBoss_C
  tracks [{
    type=MovieSceneSkeletalAnimationTrack
    sections [
     {type=MovieSceneSkeletalAnimationSection, end_frame=3942, start_frame=3599}
     {type=MovieSceneSkeletalAnimationSection, end_frame=4193, start_frame=3908}
    ]
   }]
 }]
display_rate {denominator=1, numerator=60}
tick_resolution {denominator=1, numerator=24000}
tracks []
