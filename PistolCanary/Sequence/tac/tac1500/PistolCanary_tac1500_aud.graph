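# Audio sub-sequence for the tac1500 tentacle boss sequence in PistolCanary, covering a tentacle charge-up at 60 fps display rate.
# A TentacleChargeUp audio track layers seven overlapping sound sections, each at default volume and pitch.
# An event track on the possessed PistolCanary_AudioManager fires six triggers timed close to the starts of those sounds.
# The TentacleBoss binding is present but carries no tracks.
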LevelSequence /PistolCanary/Sequence/tac/tac1500/PistolCanary_tac1500_aud
bindings [
 TentacleBoss {id=#1, possessed_class=PistolCanary_TentacleBoss_C, tracks []}
 PistolCanary_AudioManager {
  id=#2
  possessed_class=PistolCanary_AudioManager_C
  tracks [{
    type=MovieSceneEventTrack
    sections [{
      type=MovieSceneEventTriggerSection
      channels [None {
        type=MovieSceneScriptingEventChannel
        keys [
         [3602, "<Struct 'MovieSceneEvent' (0x852ff39b2c0) {}>"]
         [3798, "<Struct 'MovieSceneEvent' (0x852ff39b360) {}>"]
         [3861, "<Struct 'MovieSceneEvent' (0x852ff39b540) {}>"]
         [3938.2125, "<Struct 'MovieSceneEvent' (0x852ff39b5e0) {}>"]
         [4014, "<Struct 'MovieSceneEvent' (0x852ff39b680) {}>"]
         [4081, "<Struct 'MovieSceneEvent' (0x852ff39b860) {}>"]
        ]
       }]
     }]
   }]
 }
]
display_rate {denominator=1, numerator=60}
tick_resolution {denominator=1, numerator=24000}
tracks [{
  type=MovieSceneAudioTrack
  display_name=TentacleChargeUp
  sections [
   {
    type=MovieSceneAudioSection
    end_frame=4050
    start_frame=3599
    channels [
     Volume {type=MovieSceneScriptingFloatChannel, default=1}
     Pitch {type=MovieSceneScriptingFloatChannel, default=1}
     AttachActor {type=MovieSceneScriptingActorReferenceChannel}
    ]
   }
   {
    type=MovieSceneAudioSection
    end_frame=3957
    start_frame=3795
    channels [
     Volume {type=MovieSceneScriptingFloatChannel, default=1}
     Pitch {type=MovieSceneScriptingFloatChannel, default=1}
     AttachActor {type=MovieSceneScriptingActorReferenceChannel}
    ]
   }
   {
    type=MovieSceneAudioSection
    end_frame=3993
    start_frame=3857
    channels [
     Volume {type=MovieSceneScriptingFloatChannel, default=1}
     Pitch {type=MovieSceneScriptingFloatChannel, default=1}
     AttachActor {type=MovieSceneScriptingActorReferenceChannel}
    ]
   }
   {
    type=MovieSceneAudioSection
    end_frame=4111
    start_frame=3932
    channels [
     Volume {type=MovieSceneScriptingFloatChannel, default=1}
     Pitch {type=MovieSceneScriptingFloatChannel, default=1}
     AttachActor {type=MovieSceneScriptingActorReferenceChannel}
    ]
   }
   {
    type=MovieSceneAudioSection
    end_frame=4490
    start_frame=4069
    channels [
     Volume {type=MovieSceneScriptingFloatChannel, default=1}
     Pitch {type=MovieSceneScriptingFloatChannel, default=1}
     AttachActor {type=MovieSceneScriptingActorReferenceChannel}
    ]
   }
   {
    type=MovieSceneAudioSection
    end_frame=4133
    start_frame=4011
    channels [
     Volume {type=MovieSceneScriptingFloatChannel, default=1}
     Pitch {type=MovieSceneScriptingFloatChannel, default=1}
     AttachActor {type=MovieSceneScriptingActorReferenceChannel}
    ]
   }
   {
    type=MovieSceneAudioSection
    end_frame=4330
    start_frame=4094
    channels [
     Volume {type=MovieSceneScriptingFloatChannel, default=1}
     Pitch {type=MovieSceneScriptingFloatChannel, default=1}
     AttachActor {type=MovieSceneScriptingActorReferenceChannel}
    ]
   }
  ]
 }]
