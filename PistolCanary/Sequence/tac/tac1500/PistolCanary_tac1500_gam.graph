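# Gameplay-event Level Sequence for tac1500 in PistolCanary, driving the possessed TentacleBoss actor.
# A single event track on the boss binding holds a trigger section with six keyed events in a closely spaced series.
# It has no master tracks or other bindings, and the functions those events call are not shown.
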LevelSequence /PistolCanary/Sequence/tac/tac1500/PistolCanary_tac1500_gam
bindings [TentacleBoss {
  id=#1
  possessed_class=PistolCanary_TentacleBoss_C
  tracks [{
    type=MovieSceneEventTrack
    sections [{
      type=MovieSceneEventTriggerSection
      channels [None {
        type=MovieSceneScriptingEventChannel
        keys [
         [3749, "<Struct 'MovieSceneEvent' (0x852ff39ba40) {}>"]
         [3815, "<Struct 'MovieSceneEvent' (0x852ff39bae0) {}>"]
         [3875, "<Struct 'MovieSceneEvent' (0x852ff39b9a0) {}>"]
         [3954, "<Struct 'MovieSceneEvent' (0x852ff39b900) {}>"]
         [4025, "<Struct 'MovieSceneEvent' (0x852ff39b2c0) {}>"]
         [4090, "<Struct 'MovieSceneEvent' (0x852ff39b360) {}>"]
        ]
       }]
     }]
   }]
 }]
display_rate {denominator=1, numerator=60}
tick_resolution {denominator=1, numerator=24000}
tracks []
