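# A Level Sequence for tac1500 in PistolCanary that serves as an event trigger for the PistolCanary_PanelManager.
# It binds a PistolCanary_PanelManager_C object and holds one event track with a trigger section, with no other tracks.
# The function the event calls is not shown, so its effect on the panels is unclear.
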LevelSequence /PistolCanary/Sequence/tac/tac1500/PistolCanary_tac1500_tec
bindings [PistolCanary_PanelManager {
  id=#1
  possessed_class=PistolCanary_PanelManager_C
  tracks [{
    type=MovieSceneEventTrack
    sections [{
      type=MovieSceneEventTriggerSection
      channels [None {type=MovieSceneScriptingEventChannel, const ["<Struct 'MovieSceneEvent' (0x852ff39ae60) {}>", 1]}]
     }]
   }]
 }]
display_rate {denominator=1, numerator=60}
tick_resolution {denominator=1, numerator=24000}
tracks []
