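# VFX sequence for shot tac1500: the TentacleBoss's laser attack, from charge-up through repeated beam blasts and impacts.
# An attached charge-up Niagara system ramps via parameters alongside a post-process volume and manually offset glow pulses on the boss's body materials.
# Attached, rotation-keyed laser actors fire in staggered BeamWidth flashes, while world-placed impact and lingering systems and post-process intensity spikes punctuate the shots.
# An event track also triggers on the boss.
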LevelSequence /PistolCanary/Sequence/tac/tac1500/PistolCanary_tac1500_vfx
bindings [
 NS_TentacleAttackLaser9 {
  id=#1
  possessed_class=NiagaraActor
  tracks [
   {
    type=MovieScene3DTransformTrack
    sections [{
      type=MovieScene3DTransformSection
      channels [
       Location.X {type=MovieSceneScriptingDoubleChannel, default=0}
       Location.Y {type=MovieSceneScriptingDoubleChannel, default=0}
       Location.Z {type=MovieSceneScriptingDoubleChannel, default=0}
       Rotation.X {
        type=MovieSceneScriptingDoubleChannel
        default=179.0786
        keys [[3958,174.5893],[4023,176.6659],[4028,178.5119],[4091,161.6496],[4094,57.2917],[4097,29.9627],[4099,1.4478]]
       }
       Rotation.Y {
        type=MovieSceneScriptingDoubleChannel
        default=-56.8563
        keys [[3958,-80.3616],[4023,-82.7043],[4028,-74.0125],[4091,-88.6982],[4094,-83.9109],[4097,-77.525],[4099,-73.5558]]
       }
       Rotation.Z {
        type=MovieSceneScriptingDoubleChannel
        default=-182.8809
        keys [[3958,-175.5323],[4023,-175.8552],[4028,-177.7078],[4091,-160.7924],[4094,-56.2286],[4097,-29.1507],[4099,-0.527]]
       }
       Scale.X {type=MovieSceneScriptingDoubleChannel, default=15, const [15,1]}
       Scale.Y {type=MovieSceneScriptingDoubleChannel, default=15, const [15,1]}
       Scale.Z {type=MovieSceneScriptingDoubleChannel, default=30, const [30,1]}
      ]
     }]
   }
   {type=MovieScene3DAttachTrack, sections [{type=MovieScene3DAttachSection, end_frame=4173, start_frame=3962}]}
  ]
 }
 NS_TentacleAttackLaser5 {
  id=#2
  possessed_class=NiagaraActor
  tracks [
   {type=MovieScene3DAttachTrack, sections [{type=MovieScene3DAttachSection, end_frame=3957, start_frame=3746}]}
   {
    type=MovieScene3DTransformTrack
    sections [{
      type=MovieScene3DTransformSection
      channels [
       Location.X {type=MovieSceneScriptingDoubleChannel, default=0}
       Location.Y {type=MovieSceneScriptingDoubleChannel, default=0}
       Location.Z {type=MovieSceneScriptingDoubleChannel, default=0}
       Rotation.X {type=MovieSceneScriptingDoubleChannel, default=179.0786, const [174.5893,1]}
       Rotation.Y {type=MovieSceneScriptingDoubleChannel, default=-56.8563, const [-80.3616,1]}
       Rotation.Z {type=MovieSceneScriptingDoubleChannel, default=-182.8809, const [-175.5323,1]}
       Scale.X {type=MovieSceneScriptingDoubleChannel, default=15, const [15,1]}
       Scale.Y {type=MovieSceneScriptingDoubleChannel, default=15, const [15,1]}
       Scale.Z {type=MovieSceneScriptingDoubleChannel, default=30, const [30,1]}
      ]
     }]
   }
  ]
 }
 NiagaraComponent0 {
  id=#3
  parent_id=#1
  possessed_class=NiagaraComponent
  tracks [
   {type=MovieSceneNiagaraSystemTrack, sections [{type=MovieSceneNiagaraSystemSpawnSection, end_frame=4690, start_frame=4090}]}
   {
    type=MovieSceneNiagaraFloatParameterTrack
    display_name=BeamWidth
    sections [{type=MovieSceneFloatSection, channels [None {type=MovieSceneScriptingFloatChannel, default=0, keys [[4096,1],[4102,0],[4116,1],[4123,0]]}]}]
   }
   {
    type=MovieSceneParticleTrack
    display_name=FX System
    sections [{
      type=MovieSceneParticleSection
      channels [None {type=MovieSceneScriptingParticleChannel, keys [[4090, "<ParticleKey.ACTIVATE: 0>"], [4134, "<ParticleKey.DEACTIVATE: 1>"]]}]
     }]
   }
   {
    type=MovieSceneSpawnTrack
    sections [{
      type=MovieSceneSpawnSection
      channels [None {type=MovieSceneScriptingBoolChannel, default=false, keys [[3727,false],[4090,true],[4157,false]]}]
     }]
   }
  ]
 }
 NS_TentacleAttackLaser7 {
  id=#4
  possessed_class=NiagaraActor
  tracks [
   {type=MovieScene3DAttachTrack, sections [{type=MovieScene3DAttachSection, end_frame=4148, start_frame=3937}]}
   {
    type=MovieScene3DTransformTrack
    sections [{
      type=MovieScene3DTransformSection
      channels [
       Location.X {type=MovieSceneScriptingDoubleChannel, default=0}
       Location.Y {type=MovieSceneScriptingDoubleChannel, default=0}
       Location.Z {type=MovieSceneScriptingDoubleChannel, default=0}
       Rotation.X {type=MovieSceneScriptingDoubleChannel, default=179.0786, const [174.5893,1]}
       Rotation.Y {type=MovieSceneScriptingDoubleChannel, default=-56.8563, const [-80.3616,1]}
       Rotation.Z {type=MovieSceneScriptingDoubleChannel, default=-182.8809, const [-175.5323,1]}
       Scale.X {type=MovieSceneScriptingDoubleChannel, default=15, const [15,1]}
       Scale.Y {type=MovieSceneScriptingDoubleChannel, default=15, const [15,1]}
       Scale.Z {type=MovieSceneScriptingDoubleChannel, default=30, const [30,1]}
      ]
     }]
   }
  ]
 }
 NS_TentacleAttackImpact {
  id=#5
  possessed_class=NiagaraActor
  tracks [
   {
    type=MovieSceneSpawnTrack
    sections [{type=MovieSceneSpawnSection, channels [None {type=MovieSceneScriptingBoolChannel, default=true, keys [[3763,false],[3800,true]]}]}]
   }
   {
    type=MovieScene3DTransformTrack
    sections [{
      type=MovieScene3DTransformSection
      channels [
       Location.X {type=MovieSceneScriptingDoubleChannel, default=0, const [18108,1]}
       Location.Y {type=MovieSceneScriptingDoubleChannel, default=0, const [76368,1]}
       Location.Z {type=MovieSceneScriptingDoubleChannel, default=0, const [2664,1]}
       Rotation.X {type=MovieSceneScriptingDoubleChannel, default=0, const [0,1]}
       Rotation.Y {type=MovieSceneScriptingDoubleChannel, default=0, const [0,1]}
       Rotation.Z {type=MovieSceneScriptingDoubleChannel, default=0, const [0,1]}
       Scale.X {type=MovieSceneScriptingDoubleChannel, default=1, const [1,1]}
       Scale.Y {type=MovieSceneScriptingDoubleChannel, default=1, const [1,1]}
       Scale.Z {type=MovieSceneScriptingDoubleChannel, default=1, const [1,1]}
      ]
     }]
   }
  ]
 }
 NiagaraComponent0 {
  id=#6
  parent_id=#7
  possessed_class=NiagaraComponent
  tracks [
   {type=MovieSceneNiagaraSystemTrack, sections [{type=MovieSceneNiagaraSystemSpawnSection, end_frame=4471, start_frame=3871}]}
   {
    type=MovieSceneNiagaraFloatParameterTrack
    display_name=BeamWidth
    sections [{type=MovieSceneFloatSection, channels [None {type=MovieSceneScriptingFloatChannel, default=0, keys [[3877,1],[3883,0],[3897,1],[3904,0]]}]}]
   }
   {
    type=MovieSceneParticleTrack
    display_name=FX System
    sections [{
      type=MovieSceneParticleSection
      channels [None {type=MovieSceneScriptingParticleChannel, keys [[3871, "<ParticleKey.ACTIVATE: 0>"], [3915, "<ParticleKey.DEACTIVATE: 1>"]]}]
     }]
   }
   {
    type=MovieSceneSpawnTrack
    sections [{
      type=MovieSceneSpawnSection
      channels [None {type=MovieSceneScriptingBoolChannel, default=false, keys [[3727,false],[3871,true],[4006,false]]}]
     }]
   }
  ]
 }
 BrushComponent0 {id=#8, parent_id=#9, possessed_class=BrushComponent, tracks []}
 TentacleBoss {
  id=#10
  possessed_class=PistolCanary_TentacleBoss_C
  tracks [{
    type=MovieSceneEventTrack
    sections [{
      type=MovieSceneEventTriggerSection
      channels [None {type=MovieSceneScriptingEventChannel, const ["<Struct 'MovieSceneEvent' (0x852ff39af00) {}>", 1]}]
     }]
   }]
 }
 NiagaraComponent0 {
  id=#11
  parent_id=#12
  possessed_class=NiagaraComponent
  tracks [
   {type=MovieSceneNiagaraSystemTrack, sections [{type=MovieSceneNiagaraSystemSpawnSection, end_frame=4230, start_frame=3630}]}
   {
    type=MovieSceneNiagaraVectorParameterTrack
    display_name=ChargeScaleVec
    sections [{
      type=MovieSceneFloatVectorSection
      channels [
       Vector.X {type=MovieSceneScriptingFloatChannel, default=0.6, keys [[3632,0],[3637,0.6],[3727,1]]}
       Vector.Y {type=MovieSceneScriptingFloatChannel, default=1, const [1,2]}
       Vector.Z {type=MovieSceneScriptingFloatChannel, default=1, const [1,2]}
      ]
     }]
   }
   {
    type=MovieSceneNiagaraFloatParameterTrack
    display_name=ChargeProgress
    sections [{type=MovieSceneFloatSection, channels [None {type=MovieSceneScriptingFloatChannel, default=1, keys [[3631,0],[3663,0.3],[3722,1]]}]}]
   }
   {
    type=MovieSceneNiagaraFloatParameterTrack
    display_name=PlasmaEnabled
    sections [{type=MovieSceneFloatSection, channels [None {type=MovieSceneScriptingFloatChannel, default=0, keys [[3728,1],[3737,0]]}]}]
   }
   {
    type=MovieSceneNiagaraColorParameterTrack
    display_name=ChargeScaleColor
    sections [{
      type=MovieSceneColorSection
      channels [
       Color.R {type=MovieSceneScriptingFloatChannel, default=1}
       Color.G {type=MovieSceneScriptingFloatChannel, default=1}
       Color.B {type=MovieSceneScriptingFloatChannel, default=1}
       Color.A {type=MovieSceneScriptingFloatChannel, default=0, keys [[3646,0],[3681,1],[4130,1],[4170,0]]}
      ]
     }]
   }
   {
    type=MovieSceneNiagaraFloatParameterTrack
    display_name=ChargeProgressLightning
    sections [{type=MovieSceneFloatSection, channels [None {type=MovieSceneScriptingFloatChannel, default=0.5472, keys [[3633,0],[3689,0.5472],[3785,1]]}]}]
   }
   {
    type=MovieSceneNiagaraFloatParameterTrack
    display_name=PostProcess
    sections [{type=MovieSceneFloatSection, channels [None {type=MovieSceneScriptingFloatChannel, default=1, thinned_from=3, keys [[3634,0],[3814,1]]}]}]
   }
   {
    type=MovieSceneNiagaraFloatParameterTrack
    display_name=TipDelay
    sections [{type=MovieSceneFloatSection, channels [None {type=MovieSceneScriptingFloatChannel, default=1}]}]
   }
  ]
 }
 NS_TentacleAttackImpact3 {
  id=#13
  possessed_class=NiagaraActor
  tracks [
   {
    type=MovieSceneSpawnTrack
    sections [{type=MovieSceneSpawnSection, channels [None {type=MovieSceneScriptingBoolChannel, default=true, keys [[3763,false],[3976,true]]}]}]
   }
   {
    type=MovieScene3DTransformTrack
    sections [{
      type=MovieScene3DTransformSection
      channels [
       Location.X {type=MovieSceneScriptingDoubleChannel, default=0, keys [[3904,17834.3359],[3980,15750.7598]]}
       Location.Y {type=MovieSceneScriptingDoubleChannel, default=0, keys [[3904,83784],[3980,87204]]}
       Location.Z {type=MovieSceneScriptingDoubleChannel, default=0, keys [[3904,2030.3114],[3980,3454.0862]]}
       Rotation.X {type=MovieSceneScriptingDoubleChannel, default=0, const [0,1]}
       Rotation.Y {type=MovieSceneScriptingDoubleChannel, default=0, const [-3.127,1]}
       Rotation.Z {type=MovieSceneScriptingDoubleChannel, default=0, const [0,1]}
       Scale.X {type=MovieSceneScriptingDoubleChannel, default=1, const [1,1]}
       Scale.Y {type=MovieSceneScriptingDoubleChannel, default=1, const [1,1]}
       Scale.Z {type=MovieSceneScriptingDoubleChannel, default=1, const [1,1]}
      ]
     }]
   }
  ]
 }
 NS_TentacleAttackImpact2 {
  id=#14
  possessed_class=NiagaraActor
  tracks [
   {
    type=MovieSceneSpawnTrack
    sections [{type=MovieSceneSpawnSection, channels [None {type=MovieSceneScriptingBoolChannel, default=true, keys [[3763,false],[3893,true]]}]}]
   }
   {
    type=MovieScene3DTransformTrack
    sections [{
      type=MovieScene3DTransformSection
      channels [
       Location.X {type=MovieSceneScriptingDoubleChannel, default=0, const [12902.9717,1]}
       Location.Y {type=MovieSceneScriptingDoubleChannel, default=0, const [80544,1]}
       Location.Z {type=MovieSceneScriptingDoubleChannel, default=0, const [2836.5137,1]}
       Rotation.X {type=MovieSceneScriptingDoubleChannel, default=0, const [0,1]}
       Rotation.Y {type=MovieSceneScriptingDoubleChannel, default=0, const [-3.127,1]}
       Rotation.Z {type=MovieSceneScriptingDoubleChannel, default=0, const [0,1]}
       Scale.X {type=MovieSceneScriptingDoubleChannel, default=1, const [1,1]}
       Scale.Y {type=MovieSceneScriptingDoubleChannel, default=1, const [1,1]}
       Scale.Z {type=MovieSceneScriptingDoubleChannel, default=1, const [1,1]}
      ]
     }]
   }
  ]
 }
 NiagaraComponent0 {
  id=#15
  parent_id=#16
  possessed_class=NiagaraComponent
  tracks [
   {type=MovieSceneNiagaraSystemTrack, sections [{type=MovieSceneNiagaraSystemSpawnSection, end_frame=4351, start_frame=3751}]}
   {
    type=MovieSceneNiagaraFloatParameterTrack
    display_name=BeamWidth
    sections [{type=MovieSceneFloatSection, channels [None {type=MovieSceneScriptingFloatChannel, default=0, keys [[3777,1],[3784,0]]}]}]
   }
   {
    type=MovieSceneParticleTrack
    display_name=FX System
    sections [{
      type=MovieSceneParticleSection
      channels [None {type=MovieSceneScriptingParticleChannel, keys [[3753, "<ParticleKey.ACTIVATE: 0>"], [3795, "<ParticleKey.DEACTIVATE: 1>"]]}]
     }]
   }
   {
    type=MovieSceneSpawnTrack
    sections [{
      type=MovieSceneSpawnSection
      channels [None {type=MovieSceneScriptingBoolChannel, default=false, keys [[3727,false],[3753,true],[4006,false]]}]
     }]
   }
  ]
 }
 NiagaraComponent0 {
  id=#17
  parent_id=#18
  possessed_class=NiagaraComponent
  tracks [{type=MovieSceneNiagaraSystemTrack, sections [{type=MovieSceneNiagaraSystemSpawnSection, end_frame=4412, start_frame=3812}]}]
 }
 NS_TentacleAttackLaser8 {
  id=#19
  possessed_class=NiagaraActor
  tracks [
   {type=MovieScene3DAttachTrack, sections [{type=MovieScene3DAttachSection, end_frame=4173, start_frame=3962}]}
   {
    type=MovieScene3DTransformTrack
    sections [{
      type=MovieScene3DTransformSection
      channels [
       Location.X {type=MovieSceneScriptingDoubleChannel, default=0}
       Location.Y {type=MovieSceneScriptingDoubleChannel, default=0}
       Location.Z {type=MovieSceneScriptingDoubleChannel, default=0}
       Rotation.X {type=MovieSceneScriptingDoubleChannel, default=179.0786, keys [[3958,174.5893],[4023,176.6659],[4028,178.5119]]}
       Rotation.Y {type=MovieSceneScriptingDoubleChannel, default=-56.8563, keys [[3958,-80.3616],[4023,-82.7043],[4028,-74.0125]]}
       Rotation.Z {type=MovieSceneScriptingDoubleChannel, default=-182.8809, keys [[3958,-175.5323],[4023,-175.8552],[4028,-177.7078]]}
       Scale.X {type=MovieSceneScriptingDoubleChannel, default=15, const [15,1]}
       Scale.Y {type=MovieSceneScriptingDoubleChannel, default=15, const [15,1]}
       Scale.Z {type=MovieSceneScriptingDoubleChannel, default=30, const [30,1]}
      ]
     }]
   }
  ]
 }
 NiagaraComponent0 {
  id=#20
  parent_id=#2
  possessed_class=NiagaraComponent
  tracks [
   {type=MovieSceneNiagaraSystemTrack, sections [{type=MovieSceneNiagaraSystemSpawnSection, end_frame=4413, start_frame=3813}]}
   {
    type=MovieSceneNiagaraFloatParameterTrack
    display_name=BeamWidth
    sections [{type=MovieSceneFloatSection, channels [None {type=MovieSceneScriptingFloatChannel, default=0, keys [[3819,1],[3825,0],[3839,1],[3846,0]]}]}]
   }
   {
    type=MovieSceneParticleTrack
    display_name=FX System
    sections [{
      type=MovieSceneParticleSection
      channels [None {type=MovieSceneScriptingParticleChannel, keys [[3813, "<ParticleKey.ACTIVATE: 0>"], [3857, "<ParticleKey.DEACTIVATE: 1>"]]}]
     }]
   }
   {
    type=MovieSceneSpawnTrack
    sections [{
      type=MovieSceneSpawnSection
      channels [None {type=MovieSceneScriptingBoolChannel, default=false, keys [[3727,false],[3813,true],[4006,false]]}]
     }]
   }
  ]
 }
 PPV_PistolCanary_VFX {
  id=#9
  possessed_class=PostProcessVolume
  tracks [
   {
    type=MovieSceneFloatTrack
    display_name="Intensity (Settings)"
    sections [{
      type=MovieSceneFloatSection
      channels [None {
        type=MovieSceneScriptingFloatChannel
        default=1.5
        keys [[3799,1.5],[3801,2.5],[3833,1.5],[3886,1.5],[3888,2.5],[3920,1.5],[3972,1.5],[3974,2.5],[4006,1.5]]
       }]
     }]
   }
   {
    type=MovieSceneFloatTrack
    display_name=Blend Weight
    sections [{type=MovieSceneFloatSection, channels [None {type=MovieSceneScriptingFloatChannel, default=1, keys [[3614,0],[3720,1],[3999,1],[4026,0]]}]}]
   }
  ]
 }
 NiagaraComponent0 {
  id=#21
  parent_id=#4
  possessed_class=NiagaraComponent
  tracks [
   {type=MovieSceneNiagaraSystemTrack, sections [{type=MovieSceneNiagaraSystemSpawnSection, end_frame=4556, start_frame=3956}]}
   {
    type=MovieSceneNiagaraFloatParameterTrack
    display_name=BeamWidth
    sections [{type=MovieSceneFloatSection, channels [None {type=MovieSceneScriptingFloatChannel, default=0, keys [[3962,1],[3968,0],[3982,1],[3989,0]]}]}]
   }
   {
    type=MovieSceneParticleTrack
    display_name=FX System
    sections [{
      type=MovieSceneParticleSection
      channels [None {type=MovieSceneScriptingParticleChannel, keys [[3956, "<ParticleKey.ACTIVATE: 0>"], [4000, "<ParticleKey.DEACTIVATE: 1>"]]}]
     }]
   }
   {
    type=MovieSceneSpawnTrack
    sections [{
      type=MovieSceneSpawnSection
      channels [None {type=MovieSceneScriptingBoolChannel, default=false, keys [[3727,false],[3956,true],[4091,false]]}]
     }]
   }
  ]
 }
 NiagaraComponent0 {
  id=#22
  parent_id=#13
  possessed_class=NiagaraComponent
  tracks [{type=MovieSceneNiagaraSystemTrack, sections [{type=MovieSceneNiagaraSystemSpawnSection, end_frame=4576, start_frame=3976}]}]
 }
 NS_TentacleAttackChargeUpLaser {
  id=#12
  possessed_class=NiagaraActor
  tracks [{type=MovieScene3DAttachTrack, sections [{type=MovieScene3DAttachSection, end_frame=11296, start_frame=3608}]}]
 }
 NiagaraComponent0 {
  id=#23
  parent_id=#5
  possessed_class=NiagaraComponent
  tracks [{type=MovieSceneNiagaraSystemTrack, sections [{type=MovieSceneNiagaraSystemSpawnSection, end_frame=4400, start_frame=3800}]}]
 }
 NS_TentacleAttackLaser6 {
  id=#7
  possessed_class=NiagaraActor
  tracks [
   {
    type=MovieScene3DTransformTrack
    sections [{
      type=MovieScene3DTransformSection
      channels [
       Location.X {type=MovieSceneScriptingDoubleChannel, default=0}
       Location.Y {type=MovieSceneScriptingDoubleChannel, default=0}
       Location.Z {type=MovieSceneScriptingDoubleChannel, default=0}
       Rotation.X {type=MovieSceneScriptingDoubleChannel, default=179.0786, keys [[3873,156.3995],[3876,176.3068],[3879,177.8888]]}
       Rotation.Y {type=MovieSceneScriptingDoubleChannel, default=-56.8563, keys [[3873,-87.74],[3876,-75.8125],[3879,-64.6241]]}
       Rotation.Z {type=MovieSceneScriptingDoubleChannel, default=-182.8809, keys [[3873,-157.2829],[3876,-177.2862],[3879,-178.9594]]}
       Scale.X {type=MovieSceneScriptingDoubleChannel, default=15, const [15,1]}
       Scale.Y {type=MovieSceneScriptingDoubleChannel, default=15, const [15,1]}
       Scale.Z {type=MovieSceneScriptingDoubleChannel, default=30, const [30,1]}
      ]
     }]
   }
   {type=MovieScene3DAttachTrack, sections [{type=MovieScene3DAttachSection, end_frame=4083, start_frame=3872}]}
  ]
 }
 NS_TentacleAttackLingering {
  id=#18
  possessed_class=NiagaraActor
  tracks [{
    type=MovieScene3DTransformTrack
    sections [{
      type=MovieScene3DTransformSection
      channels [
       Location.X {type=MovieSceneScriptingDoubleChannel, default=14264, const [18036,1]}
       Location.Y {type=MovieSceneScriptingDoubleChannel, default=77756, const [76428,1]}
       Location.Z {type=MovieSceneScriptingDoubleChannel, default=3128, const [2652,1]}
       Rotation.X {type=MovieSceneScriptingDoubleChannel, default=0, const [0,1]}
       Rotation.Y {type=MovieSceneScriptingDoubleChannel, default=0, const [-13.4469,1]}
       Rotation.Z {type=MovieSceneScriptingDoubleChannel, default=0, const [0,1]}
       Scale.X {type=MovieSceneScriptingDoubleChannel, default=1, const [1,1]}
       Scale.Y {type=MovieSceneScriptingDoubleChannel, default=1, const [1,1]}
       Scale.Z {type=MovieSceneScriptingDoubleChannel, default=1, const [1,1]}
      ]
     }]
   }]
 }
 SkeletalMesh {
  id=#24
  parent_id=#10
  possessed_class=SkeletalMeshComponent
  tracks [
   {
    type=MovieSceneComponentMaterialTrack
    display_name=Material Slot: Body_A
    sections [{
      type=MovieSceneComponentMaterialParameterSection
      end_frame=4118
      start_frame=3620
      channels [
       GlowPulseEXP {type=MovieSceneScriptingFloatChannel, const [0.015,1]}
       PulseSpeed {type=MovieSceneScriptingFloatChannel, default=0, const [0,1]}
       PulseManualOffset {
        type=MovieSceneScriptingFloatChannel
        default=0
        keys [[3632,0],[3740,1],[3740.56,0],[3793,0],[3802,1],[3803,0],[3880,0],[3889,1],[3890,0],[3961,0],[3970,1],[3971,0]]
       }
       PulseRadius {type=MovieSceneScriptingFloatChannel, const [0.33,1]}
       EmissiveGlowColorR {type=MovieSceneScriptingFloatChannel, const [12,1]}
       EmissiveGlowColorG {type=MovieSceneScriptingFloatChannel, const [2.2855,1]}
       EmissiveGlowColorB {type=MovieSceneScriptingFloatChannel, default=100, const [100,1]}
       EmissiveGlowColorA {type=MovieSceneScriptingFloatChannel, const [1,1]}
      ]
     }]
   }
   {
    type=MovieSceneComponentMaterialTrack
    display_name=Material Slot: Body_B
    sections [{
      type=MovieSceneComponentMaterialParameterSection
      end_frame=4118
      start_frame=3620
      channels [
       PulseManualOffset {
        type=MovieSceneScriptingFloatChannel
        default=0
        keys [[3632,0],[3740,1],[3740.56,0],[3793,0],[3802,1],[3803,0],[3880,0],[3889,1],[3890,0],[3961,0],[3970,1],[3971,0]]
       }
       PulseSpeed {type=MovieSceneScriptingFloatChannel, default=0, const [0,1]}
       EmissiveGlowColorR {type=MovieSceneScriptingFloatChannel, const [12,1]}
       EmissiveGlowColorG {type=MovieSceneScriptingFloatChannel, const [2.2855,1]}
       EmissiveGlowColorB {type=MovieSceneScriptingFloatChannel, default=100, const [100,1]}
       EmissiveGlowColorA {type=MovieSceneScriptingFloatChannel, const [1,1]}
      ]
     }]
   }
   {
    type=MovieSceneComponentMaterialTrack
    display_name=Material Slot: Body_C
    sections [{
      type=MovieSceneComponentMaterialParameterSection
      end_frame=4118
      start_frame=3620
      channels [
       PulseSpeed {type=MovieSceneScriptingFloatChannel, default=0, const [0,1]}
       PulseManualOffset {
        type=MovieSceneScriptingFloatChannel
        default=0
        keys [[3632,0],[3740,1],[3740.56,0],[3793,0],[3802,1],[3803,0],[3880,0],[3889,1],[3890,0],[3961,0],[3970,1],[3971,0]]
       }
       EmissiveGlowColorR {type=MovieSceneScriptingFloatChannel, const [12,1]}
       EmissiveGlowColorG {type=MovieSceneScriptingFloatChannel, const [2.2855,1]}
       EmissiveGlowColorB {type=MovieSceneScriptingFloatChannel, default=100, const [100,1]}
       EmissiveGlowColorA {type=MovieSceneScriptingFloatChannel, const [1,1]}
      ]
     }]
   }
   {
    type=MovieSceneComponentMaterialTrack
    display_name=Material Slot: Body_D
    sections [{
      type=MovieSceneComponentMaterialParameterSection
      end_frame=4118
      start_frame=3620
      channels [
       PulseManualOffset {
        type=MovieSceneScriptingFloatChannel
        default=0
        keys [[3632,0],[3740,1],[3740.56,0],[3793,0],[3802,1],[3803,0],[3880,0],[3889,1],[3890,0],[3961,0],[3970,1],[3971,0]]
       }
       PulseSpeed {type=MovieSceneScriptingFloatChannel, default=0, const [0,2]}
       EmissiveGlowColorR {type=MovieSceneScriptingFloatChannel, const [12,1]}
       EmissiveGlowColorG {type=MovieSceneScriptingFloatChannel, const [2.2855,1]}
       EmissiveGlowColorB {type=MovieSceneScriptingFloatChannel, default=100, const [100,1]}
       EmissiveGlowColorA {type=MovieSceneScriptingFloatChannel, const [1,1]}
      ]
     }]
   }
  ]
 }
 NiagaraComponent0 {
  id=#25
  parent_id=#19
  possessed_class=NiagaraComponent
  tracks [
   {type=MovieSceneNiagaraSystemTrack, sections [{type=MovieSceneNiagaraSystemSpawnSection, end_frame=4622, start_frame=4022}]}
   {
    type=MovieSceneNiagaraFloatParameterTrack
    display_name=BeamWidth
    sections [{type=MovieSceneFloatSection, channels [None {type=MovieSceneScriptingFloatChannel, default=0, keys [[4028,1],[4034,0],[4048,1],[4055,0]]}]}]
   }
   {
    type=MovieSceneParticleTrack
    display_name=FX System
    sections [{
      type=MovieSceneParticleSection
      channels [None {type=MovieSceneScriptingParticleChannel, keys [[4022, "<ParticleKey.ACTIVATE: 0>"], [4066, "<ParticleKey.DEACTIVATE: 1>"]]}]
     }]
   }
   {
    type=MovieSceneSpawnTrack
    sections [{
      type=MovieSceneSpawnSection
      channels [None {type=MovieSceneScriptingBoolChannel, default=false, keys [[3727,false],[4022,true],[4157,false]]}]
     }]
   }
  ]
 }
 NiagaraComponent0 {
  id=#26
  parent_id=#14
  possessed_class=NiagaraComponent
  tracks [{type=MovieSceneNiagaraSystemTrack, sections [{type=MovieSceneNiagaraSystemSpawnSection, end_frame=4493, start_frame=3893}]}]
 }
 NS_TentacleAttackLaser4 {
  id=#16
  possessed_class=NiagaraActor
  tracks [
   {
    type=MovieScene3DTransformTrack
    sections [{
      type=MovieScene3DTransformSection
      channels [
       Location.X {type=MovieSceneScriptingDoubleChannel, default=0}
       Location.Y {type=MovieSceneScriptingDoubleChannel, default=0}
       Location.Z {type=MovieSceneScriptingDoubleChannel, default=0}
       Rotation.X {type=MovieSceneScriptingDoubleChannel, default=179.0786, keys [[3755,1.3646],[3763,42.0756]]}
       Rotation.Y {type=MovieSceneScriptingDoubleChannel, default=-56.8563, keys [[3755,-76.2228],[3763,-48.8813]]}
       Rotation.Z {type=MovieSceneScriptingDoubleChannel, default=-182.8809, keys [[3755,1.2286],[3763,-54.9181]]}
       Scale.X {type=MovieSceneScriptingDoubleChannel, default=15, const [15,1]}
       Scale.Y {type=MovieSceneScriptingDoubleChannel, default=15, const [15,1]}
       Scale.Z {type=MovieSceneScriptingDoubleChannel, default=30, const [30,1]}
      ]
     }]
   }
   {type=MovieScene3DAttachTrack, sections [{type=MovieScene3DAttachSection, end_frame=3957, start_frame=3746}]}
  ]
 }
]
display_rate {denominator=1, numerator=60}
tick_resolution {denominator=1, numerator=24000}
tracks []
